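# Unlit two-sided translucent Niagara ribbon and mesh material for an effect on the BlondeJaw glider, apparently a trail.
# Three offset samples of a panning trail texture, distorted by inky smoke, create chromatic split, blended with smoke sub-trails and tinted by hue shift and particle color.
# Opacity combines noisy dynamic-parameter dissolve with edge, depth and camera fades.
# SplineThicken and normal-directed panning noise provide world position offset math.
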Material /BRCosmetics/Gadgets/Assets/VinderTech_GliderChute/Glider_BlondeJaw/FX/M_BlondeJaw_Port
graph {
 properties {
  bAutomaticallySetUsageInEditor=false
  bUseTranslucencyVertexFog=false
  bUsedWithNiagaraMeshParticles=true
  bUsedWithNiagaraRibbons=true
  blendMode=BLEND_Translucent
  pixelDepthOffsetMode=PDOM_Legacy
  shadingModel=MSM_Unlit
  twoSided=true
  usedShadingModels=Unlit
 }
 nodes [
  n1 = ScalarParameter() {defaultValue=200, group=Spline Thicken, parameterName=SplineWidth}
  n2 = MaterialFunctionCall(WidthBase=@n1, WidthTip=@n1, Expand U or V UV Channel=StaticBool()) {materialFunction=&SplineThicken}
  n3 = VertexNormalWS()
  n4 = TextureCoordinate() {uTiling=2, vTiling=2}
  n5 = Power(Base=TextureSample(UVs=Panner(Coordinate=@n4) {speedX=0.1843, speedY=0.12}) {samplerType=SAMPLERTYPE_Grayscale, texture=&LowResBlurredNoise}.R) {constExponent=0.7}
  Add(
   A=@n2.WorldPositionOffset
   B=Multiply(A=DynamicParameter() {defaultValue=<rgba 1 1.1 1 1>, paramNames [Dissolve,DissolveEdge,WPONoise,Param4]}.WPONoise, B=Multiply(A=@n3, B=@n5))
  )
  n6 = VectorParameter() {group=Color, parameterName=BackgroundColor2, defaultValue=<rgba 0.0328 0.0804 0.15 1>}
  n7 = VectorParameter() {group=Color, parameterName=BackgroundColor1, defaultValue=<rgba 0.1117 0.0797 0.2 1>}
  n8 = LinearInterpolate(
   A=Constant3Vector() {constant=<rgba 0.0881 0.1024 0.5 1>}
   B=Constant3Vector() {constant=<rgba 0.1302 0.2854 0.5 1>}
   Alpha=NamedRerouteUsage()
  )
  n9 = TextureCoordinate() {uTiling=2}
  n10 = StaticSwitchParameter(True=MaterialFunctionCall(XY=@n9) {materialFunction=&Swizzle}.YX, False=@n9) {parameterName=Swizzled}
  n11 = Multiply(A=Constant2Vector() {r=-0.75}, B=NamedRerouteUsage())
  n12 = ComponentMask(None=TextureSample(UVs=Add(A=@n10, B=@n11)) {samplerType=SAMPLERTYPE_Grayscale, texture=&LowResBlurredNoise}.RGB) {g=true, r=true}
  n13 = Add(A=Multiply(A=@n12) {constB=0.2}) {constB=-0.1}
  n14 = TextureCoordinate()
  n15 = MaterialFunctionCall(XY=@n14) {materialFunction=&Swizzle}
  n16 = Constant2Vector() {r=-0.75}
  n17 = NamedRerouteUsage()
  n18 = Add(A=@n13, B=Add(A=StaticSwitchParameter(True=@n15.YX, False=@n14) {parameterName=Swizzled}, B=Multiply(A=@n16, B=@n17)))
  n19 = MaterialFunctionCall!(TexCoord=@n18, ClampChannel=Constant2Vector() {g=1}) {materialFunction=/Game/Effects/Fort_Effects/MaterialFunctions/MF_FX_TilingMethod}
  n20 = TextureSample!(UVs=@n18, Tex, Level=Constant() {r=5}, Apply View MipBias) {mipValueMode=TMVM_MipLevel, texture=/Game/Effects/Fort_Effects/Textures/Smoke/T_Smoke_Trail_Soft_Chromatic}
  n21 = Multiply(A=@n19.Result, B=@n20.RGB)
  n22 = TextureCoordinate()
  n23 = VectorParameter() {parameterName=TileMain, channelNames {a="", b="", g=V, r=U}, defaultValue=<rgba 0 0 0 1>}
  n24 = StaticSwitchParameter(
   True=Multiply(
    A=MaterialFunctionCall(XY=@n22) {materialFunction=&Swizzle}.YX
    B=AppendVector(A=@n23.U, B=@n23.U)
   )
   False=@n22
  ) {parameterName=Swizzled}
  n25 = MaterialFunctionCall(Float2=@n24) {materialFunction=&BreakOutFloat2Components}
  n26 = Multiply(A=ScalarParameter() {defaultValue=5, parameterName=ChromaticOffset}) {constB=0.001}
  n27 = Multiply(A=@n26) {constB=0.7}
  n28 = Add(A=@n25.R, B=@n27)
  n29 = Add(A=@n25.G, B=@n26)
  n30 = Multiply(A=Constant2Vector() {r=-0.6}, B=NamedRerouteUsage())
  n31 = TextureCoordinate() {vTiling=2}
  n32 = StaticSwitchParameter(True=MaterialFunctionCall(XY=@n31) {materialFunction=&Swizzle}.YX, False=@n31) {parameterName=Swizzled}
  n33 = Multiply(A=Constant2Vector() {r=-0.5}, B=NamedRerouteUsage())
  n34 = TextureSample!(UVs=Add(A=Multiply(A=@n12) {constB=0.1}, B=Add(A=@n32, B=@n33)), Tex, Apply View MipBias) {texture=&T_Inky_Smoke_Tile}
  n35 = Add(A=Multiply(A=ComponentMask(None=@n34.RGB) {g=true, r=true}, B=ScalarParameter() {defaultValue=0.1, parameterName=DistortionAmount}) {constB=0}) {constB=-0.15}
  n36 = Add(A=Add(A=AppendVector(A=@n28, B=@n29), B=@n30), B=@n35)
  n37 = Constant2Vector() {g=1}
  n38 = TextureObjectParameter() {
   parameterName=MainTexture
   samplerType=SAMPLERTYPE_LinearColor
   texture=/BRCosmetics/Athena/Effects/Particles/Skydiving/NS_Contrail_CubeNinja/T_CubeNinja_Trail
  }
  n39 = TextureSample(UVs=@n36, Tex=@n38) {samplerType=SAMPLERTYPE_LinearColor, texture=&T_CubeNinja_Trail}
  n40 = MaterialFunctionCall(Float3=Multiply(
    A=MaterialFunctionCall(TexCoord=@n36, ClampChannel=@n37) {materialFunction=&MF_FX_TilingMethod}.Result
    B=@n39.RGB
   )) {materialFunction=&BreakOutFloat3Components}
  n41 = MaterialFunctionCall() {materialFunction=&LinearGradient}
  n42 = Sine(None=Multiply(A=@n41.VGradient) {constB=0.5})
  n43 = Add(A=Add(A=@n24, B=@n30), B=@n35)
  n44 = Constant2Vector() {g=1}
  n45 = TextureSample(UVs=@n43, Tex=@n38) {samplerType=SAMPLERTYPE_LinearColor, texture=&T_CubeNinja_Trail}
  n46 = MaterialFunctionCall(Float3=Multiply(
    A=MaterialFunctionCall(TexCoord=@n43, ClampChannel=@n44) {materialFunction=&MF_FX_TilingMethod}.Result
    B=@n45.RGB
   )) {materialFunction=&BreakOutFloat3Components}
  n47 = MaterialFunctionCall(Float2=@n24) {materialFunction=&BreakOutFloat2Components}
  n48 = Subtract(A=@n47.R, B=@n27)
  n49 = Subtract(A=@n47.G, B=@n26)
  n50 = Add(A=Add(A=AppendVector(A=@n48, B=@n49), B=@n30), B=@n35)
  n51 = Constant2Vector() {g=1}
  n52 = TextureSample(UVs=@n50, Tex=@n38) {samplerType=SAMPLERTYPE_LinearColor, texture=&T_CubeNinja_Trail}
  n53 = MaterialFunctionCall(Float3=Multiply(
    A=MaterialFunctionCall(TexCoord=@n50, ClampChannel=@n51) {materialFunction=&MF_FX_TilingMethod}.Result
    B=@n52.RGB
   )) {materialFunction=&BreakOutFloat3Components}
  n54 = MaterialFunctionCall(
   X=LinearInterpolate(A=@n40.G, B=@n40.G, Alpha=@n42)
   Y=LinearInterpolate(A=@n46.G, B=@n46.G, Alpha=@n42)
   Z=LinearInterpolate(A=@n53.G, B=@n53.G, Alpha=@n42)
  ) {materialFunction=&MakeFloat3}
  n55 = MaterialFunctionCall(Float3=Constant3Vector() {constant=<rgba 0.3378 0.6407 1.2 1>}) {materialFunction=&BreakOutFloat3Components}
  n56 = Power(Base=@n54.Result, Exp=@n55.R)
  n57 = Desaturation(None=Add(A=Multiply(A=@n56, B=@n55.G), B=Multiply(A=@n54.Result, B=@n55.B)), Fraction=Constant() {r=-1})
  n58 = Constant() {r=1}
  n59 = TextureCoordinate() {vTiling=2}
  n60 = MaterialFunctionCall(XY=@n59) {materialFunction=&Swizzle}
  n61 = Multiply(A=Constant() {r=-0.5}, B=NamedRerouteUsage())
  n62 = Constant() {r=-0.5}
  n63 = Add(
   A=Multiply(A=@n13) {constB=0.5}
   B=Add(A=StaticSwitchParameter(True=@n60.YX, False=@n59) {parameterName=Swizzled}, B=AppendVector(A=@n61, B=@n62))
  )
  n64 = MaterialFunctionCall(TexCoord=@n63, ClampChannel=Constant2Vector() {g=1}) {materialFunction=&MF_FX_TilingMethod}
  n65 = TextureSample(UVs=@n63) {samplerType=SAMPLERTYPE_Grayscale, texture=&T_Smoke_Trail_Soft}
  n66 = Power(Base=Multiply(A=Multiply(A=@n64.Result, B=@n65.R), B=ScalarParameter() {defaultValue=1, parameterName=UseSubTrail})) {constExponent=6}
  n67 = LinearInterpolate(
   A=Add(A=Multiply(A=@n21) {constB=0.2}, B=Multiply(A=@n57) {constA=0.5, constB=0.4})
   B=Multiply(A=Constant3Vector() {constant=<rgba 0 0.175 0.5 1>}, B=@n58)
   Alpha=@n66
  )
  n68 = MaterialFunctionCall(
   Hue Shift Percentage=ScalarParameter() {parameterName=HueShift}
   Texture=Add(A=LinearInterpolate(A=@n6.RGB, B=@n7.RGB), B=Multiply(A=@n8, B=@n67))
  ) {materialFunction=&HueShift}
  n69 = ScalarParameter() {defaultValue=-0.35, parameterName=Desaturation}
  n70 = ParticleColor()
  MaterialFunctionCall(Emissive=Multiply(A=Desaturation(None=@n68.Result, Fraction=@n69), B=@n70.RGB)) {materialFunction=&MF_FX_EmissiveAdjustments}
  MaterialFunctionCall!(Sine Phase=Time(), Value 1=Constant() {r=-408.1702}, Value 2=Constant() {r=224.9242}) {materialFunction=/Engine/Functions/Engine_MaterialFunctions02/Utility/Sine_Remapped}
  n71 = Constant() {r=1}
  n72 = Add(A=Multiply(A=@n19.Result, B=@n20.G), B=Multiply(A=Desaturation(None=@n54.Result, Fraction=@n71)) {constB=0.8})
  n73 = TextureCoordinate() {vTiling=2}
  n74 = StaticSwitchParameter(True=MaterialFunctionCall(XY=@n73) {materialFunction=&Swizzle}.YX, False=@n73) {parameterName=Swizzled}
  n75 = Multiply(A=TextureSample(UVs=Panner(Coordinate=@n74)) {texture=&T_Inky_Smoke_Tile}.R) {constB=0.5}
  n76 = Constant() {r=3}
  n77 = Multiply(A=Saturate(None=Add(A=@n41.UGradient, B=@n75)), B=@n76) {constB=3}
  n78 = LinearInterpolate(A=@n76, Alpha=DynamicParameter() {defaultValue=<rgba 1 1.1 1 1>, paramNames [Dissolve,DissolveEdge,WPONoise,Param4]}.Dissolve) {constA=3, constB=-1}
  n79 = Multiply(A=Multiply(A=LinearInterpolate(A=@n72, B=@n58, Alpha=@n66) {constB=200}, B=@n70.A), B=Saturate(None=Subtract(A=@n77, B=@n78)))
  n80 = DepthFade(Opacity=@n79) {fadeDistanceDefault=50}
  n81 = ComponentMask(None=TextureCoordinate()) {g=true}
  n82 = OneMinus(None=@n81)
  n83 = Saturate(None=Multiply(A=Multiply(A=@n81, B=@n82)) {constB=5})
  Multiply(
   A=Multiply(A=ShadingPathSwitch(Default=@n80, Deferred=@n80, Forward=@n79, Mobile=@n79), B=Multiply(A=@n83))
   B=MaterialFunctionCall(Fade Length=Constant() {r=100}, Fade Offset=Constant() {r=40}) {materialFunction=&CameraDepthFade}.Result
  )
  NamedRerouteDeclaration(None=Multiply(A=Time(), B=ScalarParameter() {defaultValue=0.6, parameterName=PanForwardOrBackward})) {name=TimeInput, nodeColor=<rgba 1 0 0.0458 1>}
  n84 = TextureCoordinate()
  n85 = StaticSwitchParameter(True=MaterialFunctionCall(XY=@n84) {materialFunction=&Swizzle}.YX, False=@n84) {parameterName=Swizzled}
  n86 = Add(
   A=Multiply(A=Time(), B=ScalarParameter() {defaultValue=-0.1, group=Sine Wave, parameterName=WaveSpeed})
   B=Divide(A=ComponentMask(None=@n85) {r=true}, B=ScalarParameter() {defaultValue=0.5, group=Sine Wave, parameterName=WaveLength})
  )
  NamedRerouteDeclaration(None=ConstantBiasScale(None=Sine(None=@n86))) {name=VertexColorSine, nodeColor=<rgba 0 0.3438 1 1>}
  StaticSwitchParameter(True=@n2.WorldPositionOffset, False=Constant()) {parameterName=WPO}
  n87 = LinearInterpolate(Alpha=@n80) {constA=1, constB=1.05}
  n88 = Constant() {r=1}
  n89 = QualitySwitch(Default=@n87, Low=@n88, High=@n87, Medium=@n88, Epic=@n87)
  StaticSwitchParameter(True=Constant() {r=1}, False=ShadingPathSwitch(Default=@n89, Deferred=@n89, Forward=@n88, Mobile=@n88)) {parameterName=WPO}
 ]
}
