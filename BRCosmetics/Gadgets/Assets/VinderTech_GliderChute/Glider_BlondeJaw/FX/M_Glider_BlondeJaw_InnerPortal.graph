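# Unlit translucent Niagara sprite material for the swirling inner portal of the BlondeJaw glider FX.
# Spiral polar UVs pan cloud noise, using original-UV derivatives for mip selection to bypass the angular seam.
# The noise distorts the portal texture and drives a two-color edge glow scaled by dynamic parameter Edges_Mult.
# Particle color tints the sum; a separate alpha chain combines particle alpha, radial mask and camera depth fade.
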Material /BRCosmetics/Gadgets/Assets/VinderTech_GliderChute/Glider_BlondeJaw/FX/M_Glider_BlondeJaw_InnerPortal
graph {
 properties {
  bAutomaticallySetUsageInEditor=false
  bUsedWithNiagaraSprites=true
  blendMode=BLEND_Translucent
  shadingModel=MSM_Unlit
  twoSided=true
  usedShadingModels=Unlit
 }
 nodes [
  n1 = ParticleColor()
  n2 = ConstantBiasScale(None=TextureCoordinate()) {bias=-0.5, scale=2}
  n3 = ComponentMask(None=@n2) {r=true}
  n4 = ComponentMask(None=@n2) {g=true}
  n5 = Add(A=Multiply(A=@n3, B=@n3), B=Multiply(A=@n4, B=@n4))
  n6 = Frac(None=Divide(A=Arctangent2!(Y=@n3, X=@n4), B=Divide() {constA=710, constB=113}))
  n7 = AppendVector(A=Add(A=SquareRoot(None=@n5), B=LinearInterpolate(Alpha=@n6) {constB=2}), B=@n6)
  n8 = Constant2Vector() {g=1, r=1}
  n9 = TextureCoordinate()
  n10 = TextureSample!(UVs=Panner(Coordinate=Multiply(A=@n7, B=@n8)) {speedY=-0.55}, Tex, "DDX(UVs)"=DDX(Value=@n9), "DDY(UVs)"=DDY(Value=@n9), Apply View MipBias) {mipValueMode=TMVM_Derivative, texture=&/T_Cloud_Noise}
  n11 = SphereMask(A=TextureCoordinate(), B=Constant() {r=0.5}, Radius=Constant() {r=1}, Hardness=Constant() {r=0.05})
  n12 = Multiply(A=Power(Base=@n10.B) {constExponent=0.4}, B=Saturate(None=OneMinus(None=@n11)))
  n13 = Multiply(A=@n12) {constB=0.2}
  n14 = TextureCoordinate()
  n15 = Multiply(A=Add(A=Add(A=@n13, B=@n14)) {constB=-0.5}, B=ParticleSize()) {constB=572}
  n16 = TextureSample!(UVs=Add(A=Divide(A=@n15) {constB=572}) {constB=0.5}, Tex, Apply View MipBias) {texture=&/T_Glider_BlindeJaw_InnerPortal}
  n17 = OneMinus(None=SphereMask(A=TextureCoordinate(), B=Constant() {r=0.5}, Radius=Constant() {r=2.3}, Hardness=Constant() {r=0.05}))
  n18 = LinearInterpolate(A=Constant3Vector() {constant=<rgba 0.6875 2.0681 4 1>}, B=Constant3Vector() {constant=<rgba 0.9376 0.4871 4 1>}, Alpha=@n10.G)
  n19 = Multiply(
   A=DynamicParameter() {paramNames [Edges_Mult,Param2,Param3,Param4]}.Edges_Mult
   B=Multiply(A=Saturate(None=@n17), B=Multiply(A=@n18, B=@n10.R))
  )
  MaterialFunctionCall(Emissive=Multiply(A=@n1.RGB, B=Add(A=@n16.RGB, B=@n19))) {materialFunction=&MF_FX_EmissiveAdjustments}
  n20 = TextureCoordinate()
  n21 = Constant() {r=0.5}
  n22 = Constant() {r=0.7}
  Multiply(
   A=Multiply(A=@n1.A, B=SphereMask(A=@n20, B=@n21, Radius=@n21, Hardness=@n22))
   B=MaterialFunctionCall(Fade Length=Constant() {r=100}, Fade Offset=Constant() {r=40}) {materialFunction=&CameraDepthFade}.Result
  )
  Saturate(None=@n12)
 ]
}
