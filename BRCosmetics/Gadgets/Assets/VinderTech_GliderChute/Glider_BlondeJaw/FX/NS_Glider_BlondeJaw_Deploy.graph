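# Niagara system for the deploy effect of the BlondeJaw glider, attached to the glider's skeletal mesh.
# Burst emitters layer a portal (mesh and sprite), shatter meshes, flare/shockwave sprites with a light, and curl-noise fragments, split into variants by spawn group.
# A WingTrails emitter builds short-lived anim-trail ribbons between bone or socket pairs.
# Front-end modifiers adjust intensity for menus, while time-of-day inputs tint the portal and trails.
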NiagaraSystem /BRCosmetics/Gadgets/Assets/VinderTech_GliderChute/Glider_BlondeJaw/FX/NS_Glider_BlondeJaw_Deploy
dependencies {
 usedDataInterfaces [&NiagaraDataInterfaceColorCurve, &NiagaraDataInterfaceCurve, &NiagaraDataInterfaceSkeletalMesh]
 usedDynamicInputs [&MakeFloatFromVector]
 usedModules [
  &self:SystemSpawnScript.SystemScriptSource.SystemScriptGraph.NiagaraNodeAssignment_0.SetVariables_#1
  &SystemState
 ]
}
emitters {
 CrackShatter {
  emitterUpdateScript [
   &EmitterState()
   &SpawnBurst_Instantaneous(Spawn Time=0.05, Use Spawn Probability=false, UseLoopCountLimit=false)
   &SpawnBurst_Instantaneous(Spawn Group=1, Spawn Time=0.3, Use Spawn Probability=false, UseLoopCountLimit=false) {module=SpawnBurst_Instantaneous001}
  ]
  particleSpawnScript [
   &InitializeParticle(
    Color=&MaskLinearColorBySpawnGroup(Masked Linear Color=<rgba 1 1 1 1>, Passthrough Linear Color=<rgba 2.005 1.3327 5 0.02>)
    Lifetime=0.5
    Mesh Scale Mode=::NewEnumerator3
    Mesh Uniform Scale=8
    UsePositionOffset=false
    Write Material Random=true
   )
   &InitialMeshOrientation(
    Mesh Axis To Orient=&MaskVectorBySpawnGroup(Masked Vector=<1 0 0>, Passthrough Vector=<0 0 1>)
    Mesh Orientation Mode=::NewEnumerator2
    Orientation Vector=&MaskVectorBySpawnGroup(Masked Vector=<0 0 1>, Passthrough Vector=<1 0 0>)
    Use Rotation Vector=true
   )
   &CameraOffset(Camera Offset Amount=&MaskFloatBySpawnGroup(Masked Float=-8, Passthrough Float=30))
  ]
  particleUpdateScript [
   &ParticleState()
   &SkeletalMeshLocation(Bone Sampling Mode=::NewEnumerator4, MaskSpawnGroup=false, Skeletal Mesh=$System.Skeletal Mesh)
   &DynamicMaterialParameters(
    Index 0 Param 4=0.01
    Module.Param0WriteEnabled=true
    Module.Param1WriteEnabled=true
    Module.Param2WriteEnabled=true
    Module.Param3WriteEnabled=true
   )
   /SetVariables(Particles.Renderer2Color:LinearColor=&Multiply_LinearColorByFloat(
     Linear Color=&ColorFromCurve(ColorCurve=DataInterface() {propertyValues {
        bUseLUT=false
        AlphaCurve {keys [<at 0 0>, <at 0.3882 0.3>, {arriveTangent=-0.4904, interpMode=RCIM_Cubic, leaveTangent=-0.4904, time=0.6503, value=0.3}, <at 1 0 RCIM_Cubic>]}
        BlueCurve {keys [<at 0 4 RCIM_Cubic>,<at 0.9114 4>]}
        GreenCurve {keys [<at 0 1.8833 RCIM_Cubic>,<at 0.9114 0>]}
        RedCurve {keys [<at 0 1.5916 RCIM_Cubic>,<at 0.9114 0.1864>]}
       }})
     Float=&FN_FrontEndFloatModifier(Gameplay=1, Front End=0.7)
    )) {isSetParameters=true}
   &ScaleColor(
    Scale Alpha=&FloatFromCurve(
     FloatCurve=DataInterface() {propertyValues {Curve {keys [
         <at 0.0556 0 RCIM_Cubic>
         {arriveTangent=1.2443, interpMode=RCIM_Cubic, leaveTangent=1.2443, time=0.1833, value=0.5}
         {arriveTangent=-0.6123, interpMode=RCIM_Cubic, leaveTangent=-0.6123, time=0.377, value=0.4}
         {interpMode=RCIM_Cubic, tangentMode=RCTM_User, time=1, value=0}
        ]}}}
     Scale Curve=&FN_FrontEndFloatModifier(Gameplay=1, Front End=0.6)
    )
    Scale RGB=&VectorFromFloat(Value=1)
    ScaleA=true
    ScaleRGB=false
   )
   /SetVariables(Particles.WaveMeshSize:Vector3f=&VectorFromFloat(Value=&FloatFromCurve(FloatCurve=DataInterface() {propertyValues {Curve {keys [<at 0.6074 1 RCIM_Cubic>,<at 1 0.782>]}}}, Scale Curve=10))) {isSetParameters=true}
  ]
  rendererClasses [&NiagaraMeshRendererProperties]
  renderers [
   {rendererClass=&NiagaraMeshRendererProperties, rendererIndex=0}
   {rendererClass=&NiagaraMeshRendererProperties, rendererIndex=1}
  ]
 }
 FlareLightShockwave {
  emitterSpawnScript [/SetVariables(Emitter.DepthFade=100) {isSetParameters=true}]
  emitterUpdateScript [
   &EmitterState()
   &SpawnBurst_Instantaneous(Use Spawn Probability=false, UseLoopCountLimit=false)
   &SpawnBurst_Instantaneous(Spawn Group=1, Spawn Time=0.5, Use Spawn Probability=false, UseLoopCountLimit=false) {module=SpawnBurst_Instantaneous001}
   &SpawnBurst_Instantaneous(Spawn Group=2, Use Spawn Probability=false, UseLoopCountLimit=false) {module=SpawnBurst_Instantaneous002}
  ]
  particleSpawnScript [&InitializeParticle(
    Color=<rgba 0.2578 0.1598 1 1>
    Lifetime=&MergeHelper_FloatSwitch(0=0.4, 1=0.34, 2=0.85, 3=?, Index=&SpawnGroup(), Num=::NewEnumerator1)
    Mesh Scale Mode=::NewEnumerator3
    Mesh Uniform Scale=7
    Sprite Size Mode=::NewEnumerator3
    Uniform Sprite Size=1200
    UsePositionOffset=true
    Write Material Random=true
   )]
  particleUpdateScript [
   &ParticleState()
   &SkeletalMeshLocation(Bone Sampling Mode=::NewEnumerator4, MaskSpawnGroup=false, Skeletal Mesh=$System.Skeletal Mesh)
   &CameraOffset(Camera Offset Amount=&FloatFromCurve(
     FloatCurve=DataInterface() {propertyValues {Curve {keys [
         {arriveTangent=4.0899, interpMode=RCIM_Cubic, leaveTangent=4.0899, tangentMode=RCTM_User, time=0, value=0.5}
         {arriveTangent=-3.6426, interpMode=RCIM_Cubic, leaveTangent=-3.6426, tangentMode=RCTM_User, time=0.2, value=1}
         {arriveTangent=5.165, interpMode=RCIM_Cubic, leaveTangent=5.165, tangentMode=RCTM_User, time=0.4, value=0.3}
         <at 0.75 0>
        ]}}}
     Scale Curve=200
    ))
   &DynamicMaterialParameters(
    Index 0 Param 1=1
    Index 0 Param 2=&MergeHelper_FloatSwitch(
     0=&FloatFromCurve(FloatCurve=DataInterface() {propertyValues {Curve {keys [{interpMode=RCIM_Cubic, tangentMode=RCTM_User, time=0, value=1}, {interpMode=RCIM_Cubic, tangentMode=RCTM_User, time=1, value=0}]}}})
     1=&FloatFromCurve(FloatCurve=DataInterface() {propertyValues {Curve {keys [{interpMode=RCIM_Cubic, tangentMode=RCTM_User, time=0, value=0.3073}, {interpMode=RCIM_Cubic, tangentMode=RCTM_User, time=1, value=1}]}}})
     2=?
     3=?
     Index=&SpawnGroup()
     Num=::NewEnumerator0
    )
    Module.Index0Param0Write=true
    Module.Index0Param1Write=true
    Module.Index0Param2Write=true
    Module.Index0Param3Write=true
   )
   &ScaleSpriteSize(
    Scale Sprite Size Mode=::NewEnumerator0
    Uniform Scale Factor=&MergeHelper_FloatSwitch(
     0=&FloatFromCurve(FloatCurve=DataInterface() {propertyValues {Curve {keys [
          {arriveTangent=21.6769, interpMode=RCIM_Cubic, time=0, value=0.1}
          {arriveTangentWeight=0.2899, interpMode=RCIM_Cubic, leaveTangentWeight=0.1115, tangentMode=RCTM_User, time=0.5, value=1}
         ]}}})
     1=&FloatFromCurve(FloatCurve=DataInterface() {propertyValues {Curve {keys [
          {arriveTangent=-1.6376, interpMode=RCIM_Cubic, leaveTangent=-1.6376, tangentMode=RCTM_User, time=0, value=1}
          {arriveTangent=-0.1339, tangentMode=RCTM_User, time=1, value=0}
         ]}}})
     2=0
     3=?
     Index=&SpawnGroup()
     Num=::NewEnumerator1
    )
   )
   &ScaleColor(
    Scale Alpha=&MergeHelper_FloatSwitch(
     0=&FloatFromCurve(
      FloatCurve=DataInterface() {propertyValues {Curve {keys [
          {arriveTangent=21.6769, interpMode=RCIM_Cubic, leaveTangent=21.6769, tangentMode=RCTM_User, time=0, value=0}
          <at 0.1 1>
          {arriveTangent=-1.1111, interpMode=RCIM_Cubic, leaveTangent=-1.1111, time=0.35, value=1}
          <at 1 0>
         ]}}}
      Scale Curve=&FN_FrontEndFloatModifier(Gameplay=1, Front End=0.65)
     )
     1=&FloatFromCurve(
      FloatCurve=DataInterface() {propertyValues {Curve {keys [
          {arriveTangent=6.3883, interpMode=RCIM_Cubic, leaveTangent=6.3883, tangentMode=RCTM_User, time=0, value=0}
          <at 0.2271 1>
          {arriveTangent=-1.2938, interpMode=RCIM_Cubic, leaveTangent=-1.2938, time=0.5959, value=1}
          <at 1 0>
         ]}}}
      Scale Curve=&FN_FrontEndFloatModifier(Gameplay=1, Front End=0.65)
     )
     2=?
     3=?
     Index=&SpawnGroup()
     Num=::NewEnumerator0
    )
    ScaleA=true
    ScaleRGB=true
   )
   &FN_Light_Attributes(
    Color=<rgba 0.0521 0.124 2 1>
    Create Color Attribute=true
    Create Exponent Attribute=true
    Create Radius Attribute=true
    Exponent=1
    Radius=300
    Scale Color=&FloatFromCurve(
     FloatCurve=DataInterface() {propertyValues {Curve {keys [
         <at 0 0 RCIM_Cubic>
         {arriveTangent=1.6051, interpMode=RCIM_Cubic, leaveTangent=1.6051, time=0.2, value=1}
         {arriveTangent=-1.25, interpMode=RCIM_Cubic, leaveTangent=-1.25, time=0.623, value=1}
         <at 1 0>
        ]}}}
     Scale Curve=26
    )
   )
  ]
  rendererClasses [&NiagaraSpriteRendererProperties, &NiagaraLightRendererProperties]
  renderers [
   {rendererClass=&NiagaraSpriteRendererProperties}
   {rendererClass=&NiagaraLightRendererProperties, rendererIndex=1}
  ]
 }
 Portal {
  emitterUpdateScript [
   &EmitterState()
   &SpawnBurst_Instantaneous(Use Spawn Probability=false, UseLoopCountLimit=false)
  ]
  particleSpawnScript [
   &InitializeParticle(
    Lifetime=0.8
    Mesh Scale Mode=::NewEnumerator3
    Mesh Uniform Scale=0.7
    Sprite Size Mode=::NewEnumerator3
    Uniform Sprite Size=450
    UsePositionOffset=false
    Write Material Random=true
   )
   &InitialMeshOrientation(Mesh Orientation Mode=::NewEnumerator2, Orientation Vector=<-1 0 0>, Use Rotation Vector=true)
  ]
  particleUpdateScript [
   &ParticleState()
   &SkeletalMeshLocation(Bone Sampling Mode=::NewEnumerator4, MaskSpawnGroup=false, Skeletal Mesh=$System.Skeletal Mesh)
   &SpriteFacingAndAlignment(
    Sprite Facing=&RotateVectorByQuaternion(Vector To Rotate=<0 0 1>, Quaternion=$Particles.SkeletalMeshLocation.SampledRotation)
    WriteAlignment=false
    WriteFacing=true
   )
   &ScaleSpriteSize(
    Uniform Curve Scale=1.27
    Uniform Curve Sprite Scale=DataInterface() {propertyValues {Curve {keys [
        {interpMode=RCIM_Cubic, leaveTangentWeight=0.04, tangentMode=RCTM_User, time=0, value=0}
        {
         arriveTangent=1.1375
         arriveTangentWeight=0.0618
         interpMode=RCIM_Cubic
         leaveTangent=1.1375
         leaveTangentWeight=0.3757
         tangentMode=RCTM_User
         time=0.2237
         value=1
        }
        {
         arriveTangent=-1.3132
         arriveTangentWeight=0.3683
         interpMode=RCIM_Cubic
         leaveTangent=-1.3132
         leaveTangentWeight=0.0757
         tangentMode=RCTM_User
         time=0.5799
         value=1
        }
        {arriveTangent=-3.7805, interpMode=RCIM_Cubic, leaveTangent=-3.7805, tangentMode=RCTM_User, time=0.8, value=0.5359}
        {arriveTangent=-1.9655, tangentMode=RCTM_User, time=0.9719, value=0}
       ]}}}
   )
   &ScaleMeshSize(Scale Factor=&VectorFromFloat(Value=&FloatFromCurve(
      FloatCurve=DataInterface() {propertyValues {Curve {keys [
          {interpMode=RCIM_Cubic, leaveTangentWeight=0.04, tangentMode=RCTM_User, time=0, value=0}
          {
           arriveTangent=0.2636
           arriveTangentWeight=0.04
           interpMode=RCIM_Cubic
           leaveTangent=0.2636
           leaveTangentWeight=0.2433
           tangentMode=RCTM_User
           time=0.17
           value=1
          }
          {
           arriveTangent=-1.1364
           arriveTangentWeight=0.2433
           interpMode=RCIM_Cubic
           leaveTangent=-1.1364
           leaveTangentWeight=0.05
           tangentMode=RCTM_User
           time=0.75
           value=1
          }
          {tangentMode=RCTM_User, time=1, value=0}
         ]}}}
      Scale Curve=1.104
     )))
   &ScaleColor(
    Scale Alpha=&FloatFromCurve(FloatCurve=DataInterface() {propertyValues {Curve {keys [
         <at 0.0249 0 RCIM_Cubic>
         {interpMode=RCIM_Cubic, tangentMode=RCTM_User, time=0.12, value=1}
         {interpMode=RCIM_Cubic, leaveTangentWeight=1.182, tangentMode=RCTM_User, time=0.8482, value=1}
         {interpMode=RCIM_Cubic, tangentMode=RCTM_Break, time=0.9419, value=0}
        ]}}})
    Scale RGB=&Multiply_Vector(
     A=&Multiply_VectorByFloat(Vector=<1.7 1.7 1.7>, Float=&FN_FrontEndFloatModifier(Gameplay=1, Front End=0.7))
     B=&VectorFromFloat(Value=&FNTimeOfDayScalar(Night Scale=0.66, Morning Scale=0.85, Evening Scale=0.7))
    )
    ScaleA=true
    ScaleRGB=true
   )
   /SetVariables(Particles.InsideColor:LinearColor=&MakeLinearColorFromVectorAndFloat(
     "Vector (RGB)"=&MakeVectorFromLinearColorRGB(Color=&ColorFromCurve(
       ColorCurve=DataInterface() {propertyValues {
         AlphaCurve {keys [<at 0.13 1>,<at 0.173 1>,<at 0.25 1>,<at 0.375 1>,<at 0.5 1>,<at 0.625 1>,<at 0.75 1>,<at 0.873 1>,<at 0.9078 1>]}
         BlueCurve {keys [<at 0 0.34>,<at 0.125 0.34>,<at 0.166 0.295>,<at 0.25 0.3688>,<at 0.5 0.4148>,<at 0.7338 0.3411>,<at 0.875 0.295>,<at 0.9078 0.34>,<at 1 0.34>]}
         GreenCurve {keys [<at 0 0.2452>,<at 0.125 0.2452>,<at 0.166 0.3067>,<at 0.25 0.3833>,<at 0.5 0.4313>,<at 0.7338 0.3546>,<at 0.875 0.3067>,<at 0.9078 0.2452>,<at 1 0.2452>]}
         RedCurve {keys [<at 0 0.1488>,<at 0.125 0.1488>,<at 0.166 0.32>,<at 0.25 0.4>,<at 0.5 0.45>,<at 0.7338 0.37>,<at 0.875 0.32>,<at 0.9078 0.1488>,<at 1 0.1488>]}
        }}
       CurveIndex=Linked() {linkedVariable {name=NPC.FortniteNPC.TimeOfDay_0-1, type=NiagaraFloat}}
      ))
     "Float (Alpha)"=&FloatFromCurve(
      FloatCurve=DataInterface() {propertyValues {Curve {keys [<at 0.8785 1 RCIM_Cubic>,<at 0.9056 0>]}}}
      Scale Curve=&FN_FrontEndFloatModifier(Gameplay=1, Front End=0.7)
     )
    )) {isSetParameters=true}
   &DynamicMaterialParameters(
    Index 0 Param 1=1
    Index 0 Param 3=&FloatFromCurve(FloatCurve=DataInterface() {propertyValues {Curve {keys [
         <at 0 -140 RCIM_Cubic>
         {arriveTangent=458.2565, interpMode=RCIM_Cubic, leaveTangent=458.2565, tangentMode=RCTM_User, time=0.3, value=60}
         {arriveTangent=-834.5497, interpMode=RCIM_Cubic, leaveTangent=-834.5497, tangentMode=RCTM_User, time=0.5, value=60}
         {arriveTangent=0.0003, interpMode=RCIM_Cubic, leaveTangent=0.0003, tangentMode=RCTM_User, time=1, value=-104.2876}
        ]}}})
    Module.Index0Param0Write=true
    Module.Index0Param1Write=true
    Module.Index0Param2Write=true
    Module.Index0Param3Write=true
   ) {module=DynamicMaterialParameters001}
   /SetVariables(Particles.ImageDynamic:Vector4f=&MakeVector4(
     X=&FloatFromCurve(FloatCurve=DataInterface() {propertyValues {Curve {keys [<at 0.3563 9>,<at 1 0>]}}})
     Y=0
     Z=0
     W=0
    )) {isSetParameters=true}
  ]
  rendererClasses [&NiagaraMeshRendererProperties, &NiagaraSpriteRendererProperties]
  renderers [
   {rendererClass=&NiagaraMeshRendererProperties, rendererIndex=0}
   {rendererClass=&NiagaraSpriteRendererProperties, rendererIndex=1}
  ]
 }
 TinyFrags {
  emitterUpdateScript [
   &EmitterState(Enable Distance Culling=true, Module.bUseMaxDistance=true, Module.bUseMinDistance=false, Scalability Mode=::NewEnumerator1)
   &SpawnBurst_Instantaneous(Spawn Count=17, Spawn Time=0.05, Use Spawn Probability=false, UseLoopCountLimit=false)
   &SpawnBurst_Instantaneous(Spawn Count=35, Spawn Group=1, Spawn Time=0.05, Use Spawn Probability=false, UseLoopCountLimit=false) {module=SpawnBurst_Instantaneous001}
  ]
  particleSpawnScript [
   /SetVariables(Particles.SpriteRotationRate=&RandomRangeFloat(Minimum=-1, Maximum=1, Evaluation Type=::NewEnumerator1)) {isSetParameters=true}
   &InitializeParticle(
    Color=&MaskLinearColorBySpawnGroup(Masked Linear Color=<rgba 1 1 1 1>, Passthrough Linear Color=<rgba 1.6371 12.4037 20 1>)
    Lifetime=&RandomRangeFloat(Minimum=0.5, Maximum=1, Evaluation Type=::NewEnumerator1)
    Mesh Scale Max=<0.7 0.7 1>
    Mesh Scale Min=<0.4 0.4 1>
    Mesh Scale Mode=::NewEnumerator2
    Sprite Rotation Mode=::NewEnumerator2
    Sprite Size Mode=::NewEnumerator1
    Uniform Sprite Size Max=12
    Uniform Sprite Size Min=3
    UsePositionOffset=false
    Write Material Random=true
   ) {module=InitializeParticle001}
   &SkeletalMeshLocation(
    Bone Sampling Mode=::NewEnumerator4
    MaskSpawnGroup=false
    Sampled Position Offset=&Multiply_Vector(
     A=&RandomRangeVector(Minimum=<-1 -1 -1>, Maximum=<1 1 1>)
     B=&VectorFromFloat(Value=&MaskFloatBySpawnGroup(Masked Float=50, Passthrough Float=80))
    )
    Skeletal Mesh=$System.Skeletal Mesh
   )
   &ShapeLocation(
    MaskSpawnGroup=false
    Offset Mode=::NewEnumerator0
    Override Seed=false
    Sphere Radius=&MaskFloatBySpawnGroup(Masked Float=60, Passthrough Float=100)
    UseOwnerScale=true
   ) {enabled=false}
   &AddVelocity(
    Module.Use Falloff Distance=false
    Override Seed=false
    Velocity Mode=::NewEnumerator1
    Velocity Origin=$Particles.SkeletalMeshLocation.SampledPosition
    Velocity Speed=&RandomRangeFloat(Minimum=1400, Maximum=2500)
   ) {module=AddVelocity001}
   &InitialMeshOrientation(Use Rotation Vector=false)
  ]
  particleUpdateScript [
   &ParticleState()
   &CurlNoiseForce(
    Bias Noise=false
    Noise Frequency=&MaskFloatBySpawnGroup(Masked Float=2, Passthrough Float=4)
    Noise Strength=&MaskFloatBySpawnGroup(Masked Float=1000, Passthrough Float=5000)
    Pan Noise=false
   )
   &Drag(
    Drag=&MaskFloatBySpawnGroup(Masked Float=4, Passthrough Float=10)
    Use Linear Drag=true
    Use Rotational Drag=true
   )
   &SolveForcesAndVelocity(Clamp Velocity=false, Limit Acceleration=false)
   &ScaleSpriteSize(Scale Factor=&Vector2DFromFloat(Value=&FloatFromCurve(
      FloatCurve=DataInterface() {propertyValues {Curve {keys [
          <at -0.0044 0.2654>
          {arriveTangent=0.7558, interpMode=RCIM_Cubic, leaveTangent=0.7558, time=0.157, value=0.5261}
          {arriveTangent=-0.6241, interpMode=RCIM_Cubic, leaveTangent=-0.6241, time=0.7034, value=0.8003}
          <at 1 0>
         ]}}}
      Scale Curve=2
     )))
   &ScaleColor(
    Scale Alpha=&FloatFromCurve(FloatCurve=DataInterface() {propertyValues {Curve {keys [
         <at 0 0 RCIM_Cubic>
         {arriveTangent=1.4286, interpMode=RCIM_Cubic, leaveTangent=1.4286, time=0.2, value=1}
         {arriveTangent=-1.25, interpMode=RCIM_Cubic, leaveTangent=-1.25, time=0.7, value=1}
         <at 1 0>
        ]}}})
    Scale RGB=&MaskVectorBySpawnGroup(
     Masked Vector=&MakeVectorFromLinearColorRGB(Color=&ColorFromCurve(ColorCurve=DataInterface() {propertyValues {
         AlphaCurve {keys [<at 0 1 RCIM_Cubic>,<at 1 0 RCIM_Cubic>]}
         BlueCurve {keys [<at 0.1928 50 RCIM_Cubic>,<at 1 1 RCIM_Cubic>]}
         GreenCurve {keys [<at 0.1928 50 RCIM_Cubic>,<at 1 1 RCIM_Cubic>]}
         RedCurve {keys [<at 0.1928 50 RCIM_Cubic>,<at 1 1 RCIM_Cubic>]}
        }}))
     Passthrough Vector=&VectorFromFloat(Value=&FN_FrontEndFloatModifier(Gameplay=1, Front End=0.7))
    )
    ScaleA=true
    ScaleRGB=true
   )
   &UpdateMeshOrientation(Rotation Rate=0.5, Rotation Vector=&RandomVector(Vector Scale=1)) {module=UpdateMeshOrientation001}
   &ScaleMeshSize(Scale Factor=&VectorFromFloat(Value=&FloatFromCurve(
      FloatCurve=DataInterface() {propertyValues {Curve {keys [
          {arriveTangent=9.5937, interpMode=RCIM_Cubic, leaveTangent=9.5937, tangentMode=RCTM_User, time=0, value=0}
          <at 0.2 1>
          {arriveTangentWeight=0.2134, interpMode=RCIM_Cubic, leaveTangentWeight=0.2134, tangentMode=RCTM_User, time=0.6, value=1}
          {arriveTangent=-5.5237, interpMode=RCIM_Cubic, leaveTangent=-5.5237, tangentMode=RCTM_User, time=1, value=0}
         ]}}}
      Scale Curve=&Multiply_Float(A=4, B=1)
     )))
   &SpriteRotationRate(Rotation Rate=$Particles.SpriteRotationRate)
  ]
  rendererClasses [&NiagaraMeshRendererProperties, &NiagaraSpriteRendererProperties]
  renderers [
   {rendererClass=&NiagaraMeshRendererProperties, rendererIndex=0}
   {rendererClass=&NiagaraSpriteRendererProperties, rendererIndex=1}
  ]
 }
 WingTrails {
  emitterUpdateScript [
   &EmitterState(Life Cycle Mode=::NewEnumerator1, Loop Behavior=::NewEnumerator1, Loop Duration=0.8, UseLoopDelay=false)
   &FN_SpawnAnimTrail_v2(Number of Anim Trails=4)
  ]
  particleSpawnScript [
   /Game/Effects/Niagara/Modules/AnimTrail/FN_InitializeAnimTrail/FN_InitializeAnimTrailBoneIndecies(First Bone or Socket Index=1)
   &FN_InitializeAnimTrailBoneIndecies(AnimTrail Selection=1, First Bone or Socket Index=3, Second Bone or Socket Index=2) {module=FN_InitializeAnimTrailBoneIndecies001}
   &FN_InitializeAnimTrail(AnimTrailWidth=2, Color=<rgba 1.5625 2.1988 3 1>, Lifetime=0.2, SkeletalMesh=$System.Skeletal Mesh) {module=FN_InitializeAnimTrail001}
   &AddVelocity(Velocity=&RotateVectorByQuaternion(
     Vector To Rotate=&MakeVector(
      X=&FN_FrontEndFloatModifier(Gameplay=550, Front End=-550)
      Y=&MaskFloatBySpawnGroup(Masked Float=-20, Passthrough Float=20)
      Z=0
     )
     Quaternion=<0 0 1 1>
    ))
  ]
  particleUpdateScript [
   &InheritSourceMovement(
    Applied Position Scale=&Multiply_VectorByFloat(
     Vector=<1 1 1>
     Float=&FloatFromCurve(FloatCurve=DataInterface() {propertyValues {Curve {keys [
          {interpMode=RCIM_Cubic, tangentMode=RCTM_User, time=0.75, value=1}
          {arriveTangent=-0.3795, interpMode=RCIM_Cubic, leaveTangent=-0.3795, tangentMode=RCTM_User, time=1, value=0.9}
         ]}}})
    )
    ApplyForce=false
    ApplyPosition=true
    ApplyPositionFalloff=false
   )
   &SolveForcesAndVelocity(Clamp Velocity=false, Limit Acceleration=false)
   &FN_UpdateAnimTrail(SkeletalMesh=$System.Skeletal Mesh) {module=FN_UpdateAnimTrail001}
   &ScaleColor(
    Scale Alpha=&FloatFromCurve(
     FloatCurve=DataInterface() {propertyValues {Curve {keys [<at 0.2 1 RCIM_Cubic>,<at 0.75 0 RCIM_Cubic>]}}}
     CurveIndex=$System.LoopedAge
     Scale Curve=&FN_FrontEndFloatModifier(Gameplay=1, Front End=0.7)
    )
    Scale RGB=&MakeVectorFromLinearColorRGB(Color=&FNTimeOfDayColor(Night Color=<rgba 0.65 0.65 0.65 1>, Morning Color=<rgba 0.9 0.9 0.9 1>, Day Color=<rgba 1 1 1 1>, Evening Color=<rgba 0.8 0.8 0.8 1>))
    ScaleA=true
    ScaleRGB=true
   )
   &DynamicMaterialParameters(
    Index 0 Param 1=1
    Index 0 Param 2=&RandomRangeFloat(Minimum=0, Maximum=1)
    Index 0 Param 3=2
    Index 0 Param 4=-5
    Index 1 Param 1=2
    Index 1 Param 2=1
    Module.Index0Param0Write=true
    Module.Index0Param1Write=true
    Module.Index0Param2Write=true
    Module.Index0Param3Write=true
    Module.Index1Param0Write=true
    Module.Index1Param1Write=true
    Module.Index1Param2Write=true
    Module.Index1Param3Write=true
    Write Parameter Index 1=true
   ) {module=DynamicMaterialParameters001}
  ]
  rendererClasses [&NiagaraRibbonRendererProperties]
  renderers [{rendererClass=&NiagaraRibbonRendererProperties}]
 }
}
summary {
 systemName=NS_Glider_BlondeJaw_Deploy
 emitters [
  {emitterName=CrackShatter, rendererClasses [&NiagaraMeshRendererProperties]}
  {emitterName=TinyFrags, rendererClasses [&NiagaraMeshRendererProperties, &NiagaraSpriteRendererProperties]}
  {
   emitterName=FlareLightShockwave
   rendererClasses [&NiagaraSpriteRendererProperties, &NiagaraLightRendererProperties]
  }
  {emitterName=Portal, rendererClasses [&NiagaraMeshRendererProperties, &NiagaraSpriteRendererProperties]}
  {emitterName=WingTrails, rendererClasses [&NiagaraRibbonRendererProperties]}
 ]
 userVariables [
  User.bIsFrontEnd {type=NiagaraBool}
  User.bIsFrontEndPreview {type=NiagaraBool}
  User.bIsFullyDeployed {type=NiagaraBool}
  User.ForwardDot {type=NiagaraFloat, defaultValue=0, description=""}
  User.IsFrontEnd {type=NiagaraBool}
  User.PortalSize {type=NiagaraFloat, defaultValue=1, description=""}
  User.RibbonIDOffset {type=NiagaraInt32}
 ]
}
