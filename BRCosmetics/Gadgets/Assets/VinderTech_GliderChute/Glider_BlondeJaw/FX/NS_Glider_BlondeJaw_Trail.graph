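# Niagara trail effect for the BlondeJaw glider, combining pollen dust sprites and wing ribbons attached to its skeletal mesh.
# PollenDust spawns puffs at two sockets, with rate and velocity driven by User.ForwardDot.
# WingTrails builds four anim-trail ribbons between socket pairs, their gameplay opacity scaled by a system timeline.
# Front-end switches adjust direction and opacity, time of day tints both, and a GPU PollenParticles emitter is disabled.
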NiagaraSystem /BRCosmetics/Gadgets/Assets/VinderTech_GliderChute/Glider_BlondeJaw/FX/NS_Glider_BlondeJaw_Trail
dependencies {
 usedDataInterfaces [
  &NiagaraDataInterfaceCurve
  &NiagaraDataInterfaceArrayInt32
  &NiagaraDataInterfaceVectorCurve
  &NiagaraDataInterfaceColorCurve
  &NiagaraDataInterfaceVector2DCurve
  &NiagaraDataInterfaceSkeletalMesh
  &NiagaraDataInterfaceArrayFloat
 ]
 usedDynamicInputs [
  &SetBoolByFloatComparison
  &RemapRange
  &SelectFloatFromArray
  &MakeIntFromFloat
  &NormalizeVector
  &Add_Vector
  &TransformVector
 ]
 usedModules [
  &self:SystemSpawnScript.SystemScriptSource.SystemScriptGraph.NiagaraNodeAssignment_0.SetVariables_#1
  &SystemState
  &Timeline
  &FN_Integrator
  &InterpolateOverTime
 ]
}
emitters {
 PollenDust {
  emitterUpdateScript [
   &EmitterState()
   &SpawnRate(
    SpawnRate=&Multiply_Float(A=14, B=&FN_FrontEndFloatModifier(Gameplay=$User.ForwardDot, Front End=1))
    Use Spawn Probability=false
   )
  ]
  particleSpawnScript [
   &FN_SetSpawnGroup(Number of Spawn Groups=2)
   &InitializeParticle(
    Color=<rgba 3 3 3 0.5>
    Lifetime=&UniformRangedFloat(Minimum=0.5, Maximum=0.7)
    Position Mode=::NewEnumerator1
    Sprite Rotation Mode=::NewEnumerator2
    Sprite Size Mode=::NewEnumerator1
    Uniform Sprite Size Max=320
    Uniform Sprite Size Min=270
    UsePositionOffset=false
    Write Material Random=true
   )
   &DynamicMaterialParameters(
    Index 0 Param 1=&RandomRangeFloat(Minimum=-2, Maximum=2)
    Index 0 Param 2=&RandomRangeFloat(Minimum=0, Maximum=32)
    Module.Index0Param0Write=true
    Module.Index0Param1Write=true
    Module.Index0Param2Write=false
    Module.Index0Param3Write=false
   )
   &CameraOffset(Camera Offset Amount=-10)
   &SkeletalMeshLocation(
    Bone / Socket Index=&SelectIntFromArray(Direct Array Index=&SpawnGroup(), Int Selection Array=DataInterface() {propertyValues {IntData [14,15,14,15]}})
    MaskSpawnGroup=false
    Mesh Sampling Type=::NewEnumerator5
    Sampled Position Offset=&MaskVectorBySpawnGroup(
     Masked Vector=&RandomRangeVector(Minimum=<-25 -10 -5>, Maximum=<25 10 5>)
     Passthrough Vector=&RandomRangeVector(Minimum=<-25 -10 -5>, Maximum=<25 10 5>)
    )
    Skeletal Mesh=$System.Skeletal Mesh
    Socket Sampling Mode=::NewEnumerator5
   ) {module=SkeletalMeshLocation001}
   &AddVelocityFromPoint(
    Module.Use Falloff Distance=false
    Origin Offset=&FN_FrontEndVector3Modifier(Gameplay=<0 -450 0>, Front End=<0 450 0>)
   )
   &AddVelocity(Velocity=&FN_FrontEndVector3Modifier(Gameplay=<0 100 0>, Front End=<0 -100 0>))
   &AddVelocityInCone(
    Cone Angle=10
    Cone Axis=&FN_FrontEndVector3Modifier(Gameplay=<0 1 0>, Front End=<0 -1 0>)
    Module.Use Velocity Falloff On Cone Axis=false
    Velocity Strength=300
   ) {module=AddVelocityInCone001}
   &ApplyInitialForces()
  ]
  particleUpdateScript [
   &ParticleState()
   &AccelerationForce(Acceleration=&MakeVector(X=0, Y=0, Z=150))
   &ScaleVelocity(Velocity Scale=&Multiply_VectorByFloat(
     Vector=&VectorFromCurve(VectorCurve=DataInterface() {propertyValues {
        XCurve {keys [{arriveTangent=3.1805, interpMode=RCIM_Cubic, leaveTangent=3.1805, tangentMode=RCTM_User, time=0, value=0}, <at 1 1.4>]}
        YCurve {keys [{arriveTangent=3.1805, interpMode=RCIM_Cubic, leaveTangent=3.1805, tangentMode=RCTM_User, time=0, value=0}, <at 1 1.4>]}
        ZCurve {keys [{arriveTangent=2.4373, interpMode=RCIM_Cubic, leaveTangent=2.4373, tangentMode=RCTM_User, time=0, value=0}, <at 1 0.7714>]}
       }})
     Float=&FN_FrontEndFloatModifier(
      Gameplay=&ClampFloat(Float=$User.ForwardDot, Min=&FN_FrontEndFloatModifier(Gameplay=0.5, Front End=0.8))
      Front End=1
     )
    ))
   &SolveForcesAndVelocity(Clamp Velocity=false, Limit Acceleration=false)
   &ScaleColor(
    Scale Alpha=&Lerp_Float(
     A=&FloatFromCurve(
      FloatCurve=DataInterface() {propertyValues {Curve {keys [
          {interpMode=RCIM_Cubic, tangentMode=RCTM_User, time=0, value=0}
          <at 0.3224 1 RCIM_Cubic>
          {interpMode=RCIM_Cubic, leaveTangent=-3.8631, time=1, value=0}
         ]}}}
      Scale Curve=&Multiply_Float(
       A=&MakeCustomFloatFromBool(Bool=$User.IsFrontend, True Float=0.17, False Float=0.3)
       B=&FNTimeOfDayScalar(Night Scale=0.4, Morning Scale=0.8, Day Scale=0.7, Evening Scale=0.8)
      )
     )
     B=&FloatFromCurve(
      FloatCurve=DataInterface() {propertyValues {Curve {keys [
          {interpMode=RCIM_Cubic, tangentMode=RCTM_User, time=0, value=0}
          <at 0.5162 1 RCIM_Cubic>
          {interpMode=RCIM_Cubic, leaveTangent=-3.8631, time=1, value=0}
         ]}}}
      Scale Curve=&Multiply_Float(
       A=&MakeCustomFloatFromBool(Bool=$User.IsFrontend, True Float=0.17, False Float=0.3)
       B=&FNTimeOfDayScalar(Night Scale=0.4, Morning Scale=0.8, Day Scale=0.7, Evening Scale=0.8)
      )
     )
     Alpha=&FN_FrontEndFloatModifier(Gameplay=0, Front End=1)
    )
    Scale RGB=&MakeVectorFromLinearColorRGB(Color=&ColorFromCurve(
      ColorCurve=DataInterface() {propertyValues {
        AlphaCurve {keys [<at 0.13 1>,<at 0.173 1>,<at 0.25 1>,<at 0.375 1>,<at 0.5 1>,<at 0.625 1>,<at 0.75 1>,<at 0.873 1>,<at 0.9078 1>]}
        BlueCurve {keys [<at 0.125 1.5312>,<at 0.166 0.2754>,<at 0.25 0.2754>,<at 0.5 0.2754>,<at 0.7338 0.2754>,<at 0.875 0.2754>,<at 0.9078 1.5312>]}
        GreenCurve {keys [<at 0.125 2.5531>,<at 0.166 0.4897>,<at 0.25 0.4897>,<at 0.5 0.4897>,<at 0.7338 0.4897>,<at 0.875 0.4897>,<at 0.9078 2.5531>]}
        RedCurve {keys [<at 0.125 3>,<at 0.166 0.7344>,<at 0.25 0.7344>,<at 0.5 0.7344>,<at 0.7338 0.7344>,<at 0.875 0.7344>,<at 0.9078 3>]}
       }}
      CurveIndex=Linked() {linkedVariable {name=NPC.FortniteNPC.TimeOfDay_0-1, type=NiagaraFloat}}
     ))
    ScaleA=true
    ScaleRGB=true
   )
   &ScaleSpriteSize(Uniform Curve Sprite Scale=DataInterface() {propertyValues {Curve {keys [
        {arriveTangent=0.6664, interpMode=RCIM_Cubic, leaveTangent=0.6664, tangentMode=RCTM_User, time=0, value=0.777}
        {interpMode=RCIM_Cubic, leaveTangent=-0.4663, time=1, value=1.1262}
       ]}}})
  ]
  rendererClasses [&NiagaraSpriteRendererProperties]
  renderers [{rendererClass=&NiagaraSpriteRendererProperties}]
 }
 PollenParticles {
  bEnabled=false
  simTarget=GPUComputeSim
  emitterUpdateScript [
   &EmitterState()
   &SpawnRate(SpawnRate=&Multiply_Float(A=7000, B=$User.ForwardDot), Use Spawn Probability=false)
  ]
  particleSpawnScript [
   &FN_SetSpawnGroup(Number of Spawn Groups=6)
   /SetVariables(Particles.PerParticleRandom:NiagaraFloat=&RandomRangeFloat(Minimum=0, Maximum=1)) {isSetParameters=true}
   &InitializeParticle(
    Color=<rgba 3 3 3 1>
    Lifetime=&Lerp_Float(A=0.5, B=0.5, Alpha=$Particles.PerParticleRandom)
    Position Mode=::NewEnumerator0
    Sprite Size Max=<2.4 2.4>
    Sprite Size Min=<2 2>
    Sprite Size Mode=::NewEnumerator2
    Write Material Random=true
   )
   &SkeletalMeshLocation(
    Bone / Socket Index=&Add_Integer(A=&SpawnGroup(), B=8)
    MaskSpawnGroup=false
    Mesh Sampling Type=::NewEnumerator5
    Orientation Sampling=::NewEnumerator1
    Sampled Position Offset=&RandomRangeVector(Minimum=<-35 -35 0>, Maximum=<35 35 0>)
    Skeletal Mesh=$System.Skeletal Mesh
    Socket Sampling Mode=::NewEnumerator5
   )
   &AddVelocityFromPoint(Module.Use Falloff Distance=false, Origin Offset=<0 -450 0>) {enabled=false}
   &AddVelocity(Velocity=&RandomRangeVector(Minimum=<-40 100 -60>, Maximum=<40 100 60>)) {enabled=false}
   &AddVelocity(Velocity=&RandomRangeVector(Minimum=<0 0 -20>, Maximum=<1 1 20>)) {enabled=false, module=AddVelocity001}
   &AddVelocityInCone(Cone Angle=10, Cone Axis=<0 1 0>, Module.Use Velocity Falloff On Cone Axis=false, Velocity Strength=300) {enabled=false, module=AddVelocityInCone001}
   &ApplyInitialForces() {enabled=false}
   &DynamicMaterialParameters(
    Index 0 Param 1=1
    Index 0 Param 2=1
    Index 0 Param 3=1
    Index 0 Param 4=1
    Module.Index0Param0Write=true
    Module.Index0Param1Write=true
    Module.Index0Param2Write=true
    Module.Index0Param3Write=true
   )
  ]
  particleUpdateScript [
   &ParticleState()
   &AccelerationForce(Acceleration=&MakeVector(X=0, Y=0, Z=300)) {enabled=false}
   &ScaleVelocity(Velocity Scale=&Multiply_VectorByFloat(
     Vector=&VectorFromCurve(VectorCurve=DataInterface() {propertyValues {
        XCurve {keys [{arriveTangent=2.6822, interpMode=RCIM_Cubic, leaveTangent=2.6822, tangentMode=RCTM_User, time=0, value=0}, <at 1 1>]}
        YCurve {keys [{arriveTangent=2.6822, interpMode=RCIM_Cubic, leaveTangent=2.6822, tangentMode=RCTM_User, time=0, value=0}, <at 1 1>]}
        ZCurve {keys [{arriveTangent=2.6822, interpMode=RCIM_Cubic, leaveTangent=2.6822, tangentMode=RCTM_User, time=0, value=0}, <at 1 1>]}
       }})
     Float=&FN_FrontEndFloatModifier(Gameplay=&ClampFloat(Float=$User.ForwardDot, Min=0.5), Front End=1)
    )) {enabled=false}
   &InheritSourceMovement(
    Applied Position Scale=&VectorFromFloat(Value=0.8)
    ApplyForce=false
    ApplyPosition=true
    ApplyPositionFalloff=false
   )
   &CurlNoiseForce(
    Bias Noise=false
    Coordinate Space:ENiagaraCoordinateSpace=::Local
    Noise Frequency=10
    Noise Quality / Cost=::NewEnumerator2
    Noise Strength=&FloatFromCurve(FloatCurve=DataInterface() {propertyValues {Curve {keys [<at 0 0.3444 RCIM_Cubic>,<at 1 1>]}}}, Scale Curve=1577.6687)
    Pan Noise=true
    Pan Noise Field=<0 1.3 0>
   )
   &SolveForcesAndVelocity(Clamp Velocity=false, Limit Acceleration=false)
   &ScaleColor(
    Scale Alpha=&FloatFromCurve(
     FloatCurve=DataInterface() {propertyValues {Curve {keys [
         <at 0 0 RCIM_Cubic>
         {arriveTangent=0.2741, interpMode=RCIM_Cubic, leaveTangent=0.2741, time=0.0747, value=1}
         {arriveTangent=-2.3024, interpMode=RCIM_Cubic, leaveTangent=-2.3024, tangentMode=RCTM_User, time=1, value=0.2741}
        ]}}}
     Scale Curve=&RandomRangeFloat(Minimum=0.25, Maximum=0.3)
    )
    Scale RGB=&MakeVectorFromLinearColorRGB(Color=&ColorFromCurve(
      ColorCurve=DataInterface() {propertyValues {
        AlphaCurve {keys [<at 0.13 1>,<at 0.173 1>,<at 0.25 1>,<at 0.375 1>,<at 0.5 1>,<at 0.625 1>,<at 0.75 1>,<at 0.873 1>,<at 0.9078 1>]}
        BlueCurve {keys [<at 0 0.1111>,<at 0.125 0.1111>,<at 0.166 0.3948>,<at 0.25 0.3948>,<at 0.5 0.3948>,<at 0.7338 0.3948>,<at 0.875 0.3948>,<at 0.9078 0.1111>,<at 1 0.1111>]}
        GreenCurve {keys [<at 0 0.8687>,<at 0.125 0.8687>,<at 0.166 0.4796>,<at 0.25 0.4796>,<at 0.5 0.4796>,<at 0.7338 0.4796>,<at 0.875 0.4796>,<at 0.9078 0.8687>,<at 1 0.8687>]}
        RedCurve {keys [<at 0 1.2>,<at 0.125 1.2>,<at 0.166 0.7084>,<at 0.25 0.7084>,<at 0.5 0.7084>,<at 0.7338 0.7084>,<at 0.875 0.7084>,<at 0.9078 1.2>,<at 1 1.2>]}
       }}
      CurveIndex=Linked() {linkedVariable {name=NPC.FortniteNPC.TimeOfDay_0-1, type=NiagaraFloat}}
     ))
    ScaleA=true
    ScaleRGB=true
   )
   &ScaleSpriteSize(
    Uniform Curve Scale=0.45
    Uniform Curve Sprite Scale=DataInterface() {propertyValues {Curve {keys [
        {tangentMode=RCTM_User, time=0, value=1}
        {arriveTangent=-0.0405, interpMode=RCIM_Cubic, leaveTangent=-0.0405, tangentMode=RCTM_User, time=0.7219, value=1}
        {interpMode=RCIM_Cubic, tangentMode=RCTM_User, time=1, value=0}
       ]}}}
   )
  ]
  rendererClasses [&NiagaraSpriteRendererProperties]
  renderers [{rendererClass=&NiagaraSpriteRendererProperties}]
 }
 WingTrails {
  emitterUpdateScript [
   &EmitterState()
   &FN_SpawnAnimTrail_v2(Number of Anim Trails=4)
  ]
  particleSpawnScript [
   /Game/Effects/Niagara/Modules/AnimTrail/FN_InitializeAnimTrail/FN_InitializeAnimTrailBoneIndecies(First Bone or Socket Index=1)
   &FN_InitializeAnimTrailBoneIndecies(AnimTrail Selection=1, First Bone or Socket Index=3, Second Bone or Socket Index=2) {module=FN_InitializeAnimTrailBoneIndecies001}
   &FN_InitializeAnimTrailBoneIndecies(AnimTrail Selection=2, First Bone or Socket Index=5, Second Bone or Socket Index=4) {module=FN_InitializeAnimTrailBoneIndecies002}
   &FN_InitializeAnimTrailBoneIndecies(AnimTrail Selection=3, First Bone or Socket Index=7, Second Bone or Socket Index=6) {module=FN_InitializeAnimTrailBoneIndecies003}
   &FN_InitializeAnimTrail(
    AnimTrailWidth=1.3
    Color=&Multiply_LinearColorByFloat(
     Linear Color=<rgba 1 1 1 1>
     Float=&MergeHelper_FloatSwitch(0=1.3, 1=1.3, 2=1, 3=1, Index=&SpawnGroup(), Num=::NewEnumerator2)
    )
    Lifetime=&MergeHelper_FloatSwitch(0=0.35, 1=0.35, 2=0.2, 3=0.2, Index=&SpawnGroup(), Num=::NewEnumerator2)
    SkeletalMesh=$System.Skeletal Mesh
   ) {module=FN_InitializeAnimTrail001}
   &AddVelocity(Velocity=&RotateVectorByQuaternion(
     Vector To Rotate=&MakeVector(
      X=&FN_FrontEndFloatModifier(Gameplay=550, Front End=-550)
      Y=&MaskFloatBySpawnGroup(Masked Float=-20, Passthrough Float=20)
      Z=0
     )
     Quaternion=<0 0 1 1>
    ))
  ]
  particleUpdateScript [
   &InheritSourceMovement(
    Applied Position Scale=&Multiply_VectorByFloat(
     Vector=<1 1 1>
     Float=&FloatFromCurve(FloatCurve=DataInterface() {propertyValues {Curve {keys [
          {interpMode=RCIM_Cubic, tangentMode=RCTM_User, time=0.75, value=1}
          {arriveTangent=-0.3795, interpMode=RCIM_Cubic, leaveTangent=-0.3795, tangentMode=RCTM_User, time=1, value=0.9}
         ]}}})
    )
    ApplyForce=false
    ApplyPosition=true
    ApplyPositionFalloff=false
   )
   &SolveForcesAndVelocity(Clamp Velocity=false, Limit Acceleration=false)
   &FN_UpdateAnimTrail(SkeletalMesh=$System.Skeletal Mesh) {module=FN_UpdateAnimTrail001}
   &ScaleColor(
    Scale Alpha=&FloatFromCurve(
     FloatCurve=DataInterface() {propertyValues {Curve {keys [<at 0 1 RCIM_Cubic>,<at 0.85 0 RCIM_Cubic>]}}}
     Scale Curve=&FN_FrontEndFloatModifier(Gameplay=$System.Timeline.Time, Front End=0.45)
    )
    Scale RGB=&MakeVectorFromLinearColorRGB(Color=&FNTimeOfDayColor(
      Night Color=<rgba 0.675 0.5771 0.1793 1>
      Morning Color=<rgba 0.886 0.6584 0.4661 1>
      Day Color=<rgba 0.886 0.6584 0.4661 1>
      Evening Color=<rgba 0.886 0.6584 0.4661 1>
     ))
    ScaleA=true
    ScaleRGB=true
   )
  ]
  rendererClasses [&NiagaraRibbonRendererProperties]
  renderers [{rendererClass=&NiagaraRibbonRendererProperties}]
 }
}
summary {
 systemName=NS_Glider_BlondeJaw_Trail
 emitters [
  {emitterName=PollenDust, rendererClasses [&NiagaraSpriteRendererProperties]}
  {emitterName=WingTrails, rendererClasses [&NiagaraRibbonRendererProperties]}
  {bEnabled=false, emitterName=PollenParticles, simTarget=GPUComputeSim, rendererClasses [&NiagaraSpriteRendererProperties]}
 ]
 userVariables [
  User.bIsFrontEnd {type=NiagaraBool}
  User.bIsFrontEndPreview {type=NiagaraBool}
  User.bIsFullyDeployed {type=NiagaraBool}
  User.ForwardDot {type=NiagaraFloat, defaultValue=1, description=""}
  User.IsFrontend {type=NiagaraBool}
  User.RibbonIDOffset {type=NiagaraInt32}
  User.RightDot {type=NiagaraFloat, defaultValue=0, description=""}
 ]
}
