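# Masked character-parent material for the BlondeJaw glider, adding patterned emissive that is strongest at night.
# A texture's RGB channels mask wing, body and pattern glows, colored through gradient curve-atlas lookups and reroutes.
# Radial coordinates from pre-skinned local position drive panning noise that flickers the wing and pattern layers.
# A time-of-day scalar drives the emissive added between the parent PreFX and OpacityMask/PostFX functions, dark by day and faint at morning and evening.
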Material /BRCosmetics/Gadgets/Assets/VinderTech_GliderChute/Glider_BlondeJaw/Materials/M_BlondeJaw_Glider
graph {
 properties {
  bUseMaterialAttributes=true
  bUsedWithMeshDeformer=true
  bUsedWithMorphTargets=true
  bUsedWithSkeletalMesh=true
  blendMode=BLEND_Masked
  pixelDepthOffsetMode=PDOM_Legacy
  shadingModel=MSM_FromMaterialExpression
  usedShadingModels=Default Lit | Subsurface | Clear Coat | Subsurface Profile | Cloth
 }
 nodes [
  n1 = MaterialFunctionCall() {materialFunction=&MF_CharacterParent_2023_PreFX}
  n2 = GetMaterialAttributes(None=@n1.Result) {attributeGetTypes [#1, #2]}
  n3 = Constant() {r=0.1}
  n4 = Constant()
  n5 = Constant() {r=1}
  n6 = Add(
   A=@n2.Emissive Color
   B=Multiply(
    A=NamedRerouteUsage()
    B=MaterialFunctionCall(Value_Morning=@n3, Value_Day=@n4, Value_Evening=@n3, Value_Night=@n5) {materialFunction=&Scalar_Time_of_Day}.Result
   ) {constB=0}
  )
  MaterialFunctionCall(
   Material=MaterialFunctionCall!(Result=SetMaterialAttributes(MaterialAttributes=@n2.MaterialAttributes, Emissive Color=@n6) {attributeSetTypes [#1]}) {materialFunction=/Game/Global/Materials/Character_Parent_2023/Functions/MF_CharacterParent_2023_OpacityMask}."Switch Param (False)"
   M=@n1.M
   Pre Skinned Local Position=@n1.M
   Distance Blend=@n1.M
   Fresnel=@n1.M
  ) {materialFunction=&MF_CharacterParent_2023_PostFX}
  NamedRerouteDeclaration(None=@n2.Base Color) {name=In_Color, nodeColor=<rgba 0 1 0.1562 1>}
  n7 = TextureSampleParameter2D() {
   group=BlondeJaw_Glider
   parameterName=NightEmissive_Patterns
   samplerSource=SSM_Wrap_WorldGroupSettings
   texture=/BRCosmetics/Gadgets/Assets/VinderTech_GliderChute/Glider_BlondeJaw/Textures/T_BlondeJaw_Glider_FX_TechArt
  }
  n8 = ScalarParameter() {defaultValue=1, group=BlondeJaw_Glider, parameterName=BodyEmissive_Gamma, sliderMax=1}
  n9 = CurveAtlasRowParameter(CurveTime=Multiply(A=Power(Base=@n7.G, Exp=@n8), B=ScalarParameter() {defaultValue=0.9, group=BlondeJaw_Glider, parameterName=BodyEmissive_Gain, sliderMax=1})) {
   atlas=/BRCosmetics/Characters/Player/Male/Medium/Bodies/M_MED_BlondeJaw_Strike/Materials/CA_BlondeJaw
   curve=/BRCosmetics/Gadgets/Assets/VinderTech_GliderChute/Glider_BlondeJaw/Textures/C_BlondeJaw_Glider
   defaultValue=4
   group=BlondeJaw_Glider
   parameterName=WingsEmissive_GradientMap
  }
  n10 = Multiply(A=@n9, B=ScalarParameter() {defaultValue=1, group=BlondeJaw_Glider, parameterName=BodyEmissive_Brightness})
  n11 = CurveAtlasRowParameter(CurveTime=Desaturation(None=NamedRerouteUsage())) {
   atlas=&CA_BlondeJaw
   curve=&C_BlondeJaw_Glider
   defaultValue=4
   group=BlondeJaw_Glider
   parameterName=WingsEmissive_GradientMap
  }
  n12 = LinearInterpolate(A=NamedRerouteUsage(), B=@n11, Alpha=ScalarParameter() {group=BlondeJaw_Glider, parameterName=WingsEmissive_DesatAmt})
  n13 = Add(A=LocalPosition() {localOrigin=InstancePreSkinning}.XY, B=Constant2Vector() {g=80})
  n14 = MaterialFunctionCall(Float3=PreSkinnedLocalBounds().Half Extents) {materialFunction=&BreakOutFloat3Components}
  n15 = MaterialFunctionCall(Vector or UVs=Divide(A=VertexInterpolator(VS=@n13).PS, B=Max(A=@n14.R, B=@n14.R))) {materialFunction=&VectorToRadialValue}
  n16 = MaterialFunctionCall(X=Constant() {r=0.03}, Y=ScalarParameter() {defaultValue=-0.3, group=BlondeJaw_Glider, parameterName=PanSpeed_EmissiveVariation}) {materialFunction=&MakeFloat2}
  n17 = TextureSample(UVs=Panner(Coordinate=@n15.Radial Coordinates, Speed=@n16.Result) {speedY=-0.1}) {
   samplerSource=SSM_Wrap_WorldGroupSettings
   samplerType=SAMPLERTYPE_Grayscale
   texture=/Game/Global/Textures/Default/Noise/LowResBlurredNoiseNormalized
  }
  n18 = Multiply(
   A=Multiply(A=Multiply(A=@n7.R, B=@n12), B=ScalarParameter() {defaultValue=3, group=BlondeJaw_Glider, parameterName=WingsEmissive_Brightness})
   B=LinearInterpolate(Alpha=@n17.R) {constA=0.22}
  )
  n19 = NamedRerouteUsage()
  n20 = ScalarParameter() {defaultValue=10, group=BlondeJaw_Glider, parameterName=PanScale_Pattern}
  n21 = TextureSample(UVs=Panner(Coordinate=Multiply(A=@n15.Radial Coordinates, B=@n20) {constB=0.01}, Speed=@n16.Result) {speedY=-0.1}) {
   samplerSource=SSM_Wrap_WorldGroupSettings
   samplerType=SAMPLERTYPE_Grayscale
   texture=&LowResBlurredNoiseNormalized
  }
  n22 = Multiply(
   A=Multiply(A=Multiply(A=@n7.B, B=@n19), B=ScalarParameter() {defaultValue=3, group=BlondeJaw_Glider, parameterName=PatternEmissive_Brightness})
   B=LinearInterpolate(Alpha=@n21.R) {constA=0.1}
  )
  NamedRerouteDeclaration(None=Add(A=Multiply(A=@n10, B=@n7.G), B=Add(A=@n18, B=@n22))) {name=Out_EmissiveColor, nodeColor=<rgba 0 0.0859 1 1>}
 ]
}
